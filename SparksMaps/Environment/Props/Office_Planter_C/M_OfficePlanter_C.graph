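# Masked default-lit material for an office planter prop with leaves, built from two Office_Planter_A texture sets (A and B).
# A step on the U coordinate of the second UV channel creates LeavesMask, and its inverse creates PlanterMask, both shared through named reroutes.
# Mask-weighted sums combine the two sets' diffuse, normal and S textures, with the S result split into components.
# A similar sum uses only set A's M-texture green channel.
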Material /SparksMaps/Environment/Props/Office_Planter_C/M_OfficePlanter_C
graph {
 properties {bAutomaticallySetUsageInEditor=false, blendMode=BLEND_Masked, usedShadingModels=Default Lit}
 nodes [
  Add(
   A=Multiply(
    A=TextureSample() {texture=/Game/Environments/Asteria/Props/Office/Office_Planters_A/Textures/T_Office_Planter_A_B_D}.RGB
    B=NamedRerouteUsage()
   )
   B=Multiply(
    A=TextureSample() {texture=/Game/Environments/Asteria/Props/Office/Office_Planters_A/Textures/T_Office_Planter_A_A_D}.RGB
    B=NamedRerouteUsage()
   )
  )
  Add(
   A=Multiply(
    A=TextureSample() {samplerType=SAMPLERTYPE_Normal, texture=/Game/Environments/Asteria/Props/Office/Office_Planters_A/Textures/T_Office_Planter_A_A_N}.RGB
    B=NamedRerouteUsage()
   )
   B=Multiply(
    A=TextureSample() {samplerType=SAMPLERTYPE_Normal, texture=/Game/Environments/Asteria/Props/Office/Office_Planters_A/Textures/T_Office_Planter_A_B_N}.RGB
    B=NamedRerouteUsage()
   )
  )
  Add(
   A=Multiply(
    A=TextureSample() {texture=/Game/Environments/Asteria/Props/Office/Office_Planters_A/Textures/T_Office_Planter_A_A_M}.G
    B=NamedRerouteUsage()
   )
   B=Multiply(A=NamedRerouteUsage())
  )
  n1 = TextureSample() {texture=/Game/Environments/Asteria/Props/Office/Office_Planters_A/Textures/T_Office_Planter_A_A_S}
  n2 = NamedRerouteUsage()
  n3 = TextureSample() {texture=/Game/Environments/Asteria/Props/Office/Office_Planters_A/Textures/T_Office_Planter_A_B_S}
  n4 = NamedRerouteUsage()
  MaterialFunctionCall(Float3=Add(A=Multiply(A=@n1.RGB, B=@n2), B=Multiply(A=@n3.RGB, B=@n4))) {materialFunction=&BreakOutFloat3Components}
  n5 = Step(X=ComponentMask(None=TextureCoordinate() {coordinateIndex=1}) {r=true}) {constX=2, constY=1}
  NamedRerouteDeclaration(None=@n5) {name=LeavesMask, nodeColor=<rgba 1 0 0.7266 1>}
  NamedRerouteDeclaration(None=OneMinus(None=@n5)) {name=PlanterMask, nodeColor=<rgba 0 1 0.9531 1>}
 ]
}
